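# AnimBlueprint for the StreetOps commando body mesh, adding secondary motion to the pose copied from its attached parent.
# The pose runs through a Control Rig, hood-zip AnimDynamics and weighted bone constraints for two comm wires, then gets cached.
# When the playlist uses custom character parts, checked at initialization, output switches to that pose with the head shrunk to near zero.
# Chest and hip pouch AnimDynamics nodes are unconnected to the output.
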AnimBlueprint /BRCosmetics/Characters/Player/Male/Medium/Bodies/M_MED_Commando_StreetOps/Meshes/M_MED_Street_Ops_AnimBP
target_skeleton=&/M_MED_Street_Ops_Skeleton
graphs [
 AnimGraph {nodes [
   n1 = Anim Dynamics - Bone: dyn_chestPouch(Alpha=0.2, ExternalForce="0,0,0") {
    angularDampingOverride=0.995
    angularSpringConstant=1500
    bAngularSpring=true
    bGravityOverrideInSimSpace=true
    bOverrideAngularDamping=true
    bUseGravityOverride=true
    alphaBoolBlend {blendOption=Linear}
    boundBone {boneName=dyn_chestPouch}
    componentAppliedLinearAccClamp=<300 300 300>
    componentLinearAccScale=<1 1 1>
    componentLinearVelScale=<1 1 1>
    gravityOverride {z=-981}
    physicsBodyDefinitions [{
      sphereCollisionRadius=0
      boundBone {boneName=dyn_chestPouch}
      boxExtents=<8 8 3>
      constraintSetup {twistAxis=AxisY, angularLimitsMax=<2 0 0>, angularLimitsMin=<-10 0 0>, angularTarget=<0 0 0>}
      localJointOffset=<0 -0.5 -5>
     }]
   }
   Anim Dynamics - Bone: dyn_hipPouch_l(ComponentPose=@n1, Alpha=0.5, ExternalForce="0,0,0") {
    angularDampingOverride=0.998
    angularSpringConstant=3000
    bAngularSpring=true
    bGravityOverrideInSimSpace=true
    bOverrideAngularDamping=true
    bUseGravityOverride=true
    alphaBoolBlend {blendOption=Linear}
    boundBone {boneName=dyn_hipPouch_l}
    componentAppliedLinearAccClamp=<300 300 300>
    componentLinearAccScale=<1 1 1>
    componentLinearVelScale=<1 1 1>
    gravityOverride {z=-981}
    physicsBodyDefinitions [{
      sphereCollisionRadius=0
      boundBone {boneName=dyn_hipPouch_l}
      boxExtents=<6 6 2>
      constraintSetup {twistAxis=AxisZ, angularLimitsMax=<0 2 0>, angularLimitsMin=<0 -10 0>, angularTarget=<0 0 0>}
      localJointOffset=<-0.1 0 -4>
     }]
   }
   n2 = "Transform (Modify) Bone - Bone: head"!(ComponentPose=Local To Component(LocalPose=Use cached pose 'PreScale Cache'()), Alpha=1) -> (Pose) {scaleMode=BMM_Replace, alphaBoolBlend {blendOption=Linear}, boneToModify {boneName=head}, scale=<0.01 0.01 0.01>}
   Output Pose(Result=Blend Poses by bool(
     BlendPose_0=Component To Local(ComponentPose=@n2)
     BlendPose_1=Use cached pose 'PreScale Cache'()
     BlendTime_0=0
     BlendTime_1=0
     bActiveValue=Get GetPlayerListUsesCustomCharacterParts() -> (GetPlayerListUsesCustomCharacterParts:bool)
    ) {blendType=Linear}) {layerGroup=None}
   n3 = Control Rig(Source=Copy Pose From Mesh() {bUseAttachedParent=true}) {controlRigAssetReference {blueprintRigClass=&/M_MED_Street_Ops_CtrlRig.M_MED_Street_Ops_CtrlRig_C}}
   n4 = Anim Dynamics - Bone: dyn_hoodZip_jnt(ComponentPose=Local To Component(LocalPose=@n3), Alpha=1, ExternalForce="0,0,0") {
    angularDampingOverride=0.99
    bGravityOverrideInSimSpace=true
    bOverrideAngularDamping=true
    bUseGravityOverride=true
    alphaBoolBlend {blendOption=Linear}
    boundBone {boneName=dyn_hoodZip_jnt}
    componentAppliedLinearAccClamp=<300 300 300>
    componentLinearAccScale=<1 1 1>
    componentLinearVelScale=<1 1 1>
    gravityOverride {z=-981}
    physicsBodyDefinitions [{
      sphereCollisionRadius=0
      boundBone {boneName=dyn_hoodZip_jnt}
      boxExtents=<3 3 1>
      constraintSetup {twistAxis=AxisY, angularLimitsMax=<50 0 0>, angularLimitsMin=<-10 0 0>, angularTarget=<0 0 0>}
      localJointOffset=<0 0 -2>
     }]
   }
   n5 = "Constraint - dyn_comWireB (PPPPPP)"!(
    ComponentPose=@n4
    Alpha=1
    ConstraintWeights_0=0.3522
    ConstraintWeights_1:float=0.257
    ConstraintWeights_2:float=0.1345
    ConstraintWeights_3:float=0.109
    ConstraintWeights_4:float=0.0803
    ConstraintWeights_5:float=0.067
   ) -> (Pose) {
    boneToModify {boneName=dyn_comWireB}
    constraintSetup [
     {targetBone {boneName=spine_05}}
     {targetBone {boneName=neck_01}}
     {targetBone {boneName=neck_02}}
     {targetBone {boneName=spine_04}}
     {targetBone {boneName=head}}
     {targetBone {boneName=clavicle_l}}
    ]
    constraintWeights [1,1,1,1,1,1]
   }
   n6 = "Constraint - dyn_comWireA (PPPP)"!(
    ComponentPose=@n5
    Alpha=1
    ConstraintWeights_0=0.4643
    ConstraintWeights_1:float=0.2591
    ConstraintWeights_2:float=0.1896
    ConstraintWeights_3:float=0.087
   ) -> (Pose) {
    boneToModify {boneName=dyn_comWireA}
    constraintSetup [{targetBone {boneName=head}}, {targetBone {boneName=neck_02}}, {targetBone {boneName=neck_01}}, {targetBone {boneName=spine_05}}]
    constraintWeights [1,1,1,1]
   }
   Save cached pose 'PreScale Cache'(Pose=Component To Local(ComponentPose=@n6)) {cacheName=PreScale Cache}
  ]}
 EventGraph {nodes [
   Event BlueprintInitializeAnimation()
   > Set GetPlayerListUsesCustomCharacterParts(GetPlayerListUsesCustomCharacterParts:bool=GetPlaylistUsesCustomCharacterParts!() -> (ReturnValue:bool)) -> (Output_Get:bool)
   Event BlueprintUpdateAnimation()
  ]}
]
variables [GetPlayerListUsesCustomCharacterParts {type=boolean}]
